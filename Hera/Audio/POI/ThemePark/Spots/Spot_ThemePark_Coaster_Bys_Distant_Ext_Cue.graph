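# Sound cue for exterior theme park ambience: one-shot spots of distant roller coaster pass-bys.
# The first two quality-level slots randomly pick a quieter air-release/distant coaster set or a larger pass-by set; the last two use only three pass-bys.
# Modulators randomize pitch throughout and volume on the larger sets.
# Overrides add distance low-pass and a farthest-first voice limit, with a distant reverb send rising over the far half.
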SoundCue /Hera/Audio/POI/ThemePark/Spots/Spot_ThemePark_Coaster_Bys_Distant_Ext_Cue
root=QualityLevel_0
nodes [
 {id=Modulator_0, children [Random_0], properties {pitchMax=1.2, pitchMin=0.9, volumeMax=1, volumeMin=0.8}}
 {id=Modulator_1, children [Random_1], properties {pitchMax=1.2, pitchMin=0.9, volumeMax=1, volumeMin=1}}
 {id=Modulator_2, children [Random_2], properties {pitchMax=1.2, pitchMin=0.9, volumeMax=0.6, volumeMin=0.4}}
 {id=QualityLevel_0, children [Random_3,Random_3,Modulator_1,Modulator_1]}
 {
  id=Random_0
  children [
   WavePlayer_6
   WavePlayer_7
   WavePlayer_8
   WavePlayer_9
   WavePlayer_10
   WavePlayer_11
   WavePlayer_12
   WavePlayer_13
   WavePlayer_14
   WavePlayer_15
  ]
  properties {weights [1,1,1,1,1,1,1,1,1,1]}
 }
 {id=Random_1, children [WavePlayer_16, WavePlayer_17, WavePlayer_18], properties {weights [1,1,1]}}
 {id=Random_2, children [WavePlayer_0,WavePlayer_1,WavePlayer_2,WavePlayer_3,WavePlayer_4,WavePlayer_5], properties {weights [1,1,1,1,1,1]}}
 {id=Random_3, children [Modulator_2,Modulator_0], properties {weights [1,1]}}
 {id=WavePlayer_0, properties {soundWaveAssetPtr=&/Spot_ThemePark_Roller_Coaster_Air_Release_01}}
 {id=WavePlayer_1, properties {soundWaveAssetPtr=&/Spot_ThemePark_Roller_Coaster_Distant_01}}
 {id=WavePlayer_10, properties {soundWaveAssetPtr=&/Spot_Roller_Coaster_Bys_Distant_05}}
 {id=WavePlayer_11, properties {soundWaveAssetPtr=&/Spot_Roller_Coaster_Bys_Distant_06}}
 {id=WavePlayer_12, properties {soundWaveAssetPtr=&/Spot_Roller_Coaster_Bys_Distant_07}}
 {id=WavePlayer_13, properties {soundWaveAssetPtr=&/Spot_Roller_Coaster_Bys_Distant_08}}
 {id=WavePlayer_14, properties {soundWaveAssetPtr=&/Spot_Roller_Coaster_Bys_Distant_09}}
 {id=WavePlayer_15, properties {soundWaveAssetPtr=&/Spot_Roller_Coaster_Bys_Distant_10}}
 {id=WavePlayer_16, properties {soundWaveAssetPtr=&/Spot_Roller_Coaster_Bys_Distant_01}}
 {id=WavePlayer_17, properties {soundWaveAssetPtr=&/Spot_Roller_Coaster_Bys_Distant_02}}
 {id=WavePlayer_18, properties {soundWaveAssetPtr=&/Spot_Roller_Coaster_Bys_Distant_03}}
 {id=WavePlayer_2, properties {soundWaveAssetPtr=&/Spot_ThemePark_Roller_Coaster_Distant_02}}
 {id=WavePlayer_3, properties {soundWaveAssetPtr=&/Spot_ThemePark_Roller_Coaster_Distant_03}}
 {id=WavePlayer_4, properties {soundWaveAssetPtr=&/Spot_ThemePark_Roller_Coaster_Distant_04}}
 {id=WavePlayer_5, properties {soundWaveAssetPtr=&/Spot_ThemePark_Roller_Coaster_Distant_05}}
 {id=WavePlayer_6, properties {soundWaveAssetPtr=&Spot_Roller_Coaster_Bys_Distant_01}}
 {id=WavePlayer_7, properties {soundWaveAssetPtr=&Spot_Roller_Coaster_Bys_Distant_02}}
 {id=WavePlayer_8, properties {soundWaveAssetPtr=&Spot_Roller_Coaster_Bys_Distant_03}}
 {id=WavePlayer_9, properties {soundWaveAssetPtr=&/Spot_Roller_Coaster_Bys_Distant_04}}
]
properties {
 bOverrideAttenuation=true
 bOverrideConcurrency=true
 maxDistance=8000
 soundClassObject=&Fort_Ambient_OneShots_Outdoors
 volumeMultiplier=1
 attenuationOverrides {
  bAttenuateWithLPF=true
  bEnableLogFrequencyScaling=true
  falloffDistance=8000
  lPFFrequencyAtMax=500
  lPFRadiusMax=8000
  attenuationShapeExtents {x=0}
 }
 concurrencyOverrides {maxCount=2, resolutionRule=StopFarthestThenPreventNew}
 soundSubmixSends [{
   maxSendDistance=8000
   minSendLevel=0
   sendLevelControlMethod=CustomCurve
   soundSubmix=&Fortnite_Reverb_Distant
   customSendLevelCurve {editorCurveData {keys [<at 0 0>,<at 0.5 0>,<at 1 0.6>]}}
  }]
}
