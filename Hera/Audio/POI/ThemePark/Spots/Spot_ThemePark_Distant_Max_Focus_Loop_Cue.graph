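# Theme park distant ambience cue: a looping bed with randomly delayed, modulated organ and roller coaster one-shots, limited to one instance.
# A quality node selects between two mixes.
# The first two levels pair a max-focus bed with full one-shot pools under no-focus attenuation.
# The last two use smaller pools sharing the bed's max-focus attenuation, plus a delayed, enveloped copy of the bed with close no-focus attenuation.
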SoundCue /Hera/Audio/POI/ThemePark/Spots/Spot_ThemePark_Distant_Max_Focus_Loop_Cue
root=QualityLevel_0
nodes [
 {id=Attenuation_1, children [Mixer_1], properties {attenuationSettings=&/Spot_ThemePark_Distant_Max_Focus_Att}}
 {id=Attenuation_2, children [Delay_3], properties {attenuationSettings=&/Spot_ThemePark_Close_No_Focus_Att}}
 {id=Attenuation_3, children [Mixer_2], properties {attenuationSettings=&Spot_ThemePark_Distant_Max_Focus_Att}}
 {id=Attenuation_4, children [Mixer_4], properties {attenuationSettings=&/Spot_ThemePark_Distant_NoFocus_Att}}
 {id=Delay_0, children [Modulator_1], properties {delayMax=8, delayMin=6}}
 {id=Delay_1, children [Modulator_2], properties {delayMax=12, delayMin=3}}
 {id=Delay_3, children [Enveloper_1], properties {delayMax=1, delayMin=1}}
 {id=Delay_4, children [Modulator_4], properties {delayMax=12, delayMin=3}}
 {id=Delay_5, children [Modulator_5], properties {delayMax=8, delayMin=6}}
 {
  id=Enveloper_1
  children [Modulator_3]
  properties {pitchCurve {editorCurveData {keys [{time=0}]}}, volumeCurve {editorCurveData {keys [<at 0 0>,<at 0.2 0.0769>,<at 0.75 0.5>]}}}
 }
 {id=Looping_0, children [Delay_0]}
 {id=Looping_1, children [Delay_1]}
 {id=Looping_2, children [Delay_4]}
 {id=Looping_3, children [Delay_5]}
 {id=Mixer_0, children [null, Attenuation_1, Attenuation_4], properties {inputVolume [0.75,1,1]}}
 {id=Mixer_1, children [Modulator_0], properties {inputVolume [1]}}
 {id=Mixer_2, children [Modulator_3,Looping_3,Looping_2], properties {inputVolume [1,1,1]}}
 {id=Mixer_3, children [Attenuation_2, Attenuation_3], properties {inputVolume [0.75,1]}}
 {id=Mixer_4, children [Looping_0,Looping_1], properties {inputVolume [1,1]}}
 {id=Modulator_0, children [WavePlayer_1], properties {pitchMax=1}}
 {id=Modulator_1, children [Random_0], properties {pitchMax=1.1, pitchMin=1, volumeMax=1.6, volumeMin=1.2}}
 {id=Modulator_2, children [Random_1], properties {volumeMax=1.2, volumeMin=1}}
 {id=Modulator_3, children [WavePlayer_18], properties {pitchMax=1}}
 {id=Modulator_4, children [Random_3], properties {volumeMax=1.2, volumeMin=1}}
 {id=Modulator_5, children [Random_2], properties {volumeMax=1.2, volumeMin=1}}
 {id=QualityLevel_0, children [Mixer_0,Mixer_0,Mixer_3,Mixer_3]}
 {id=Random_0, children [WavePlayer_2,WavePlayer_3,WavePlayer_4,WavePlayer_5,WavePlayer_6,WavePlayer_7], properties {weights [1,1,1,1,1,1]}}
 {
  id=Random_1
  children [
   WavePlayer_8
   WavePlayer_9
   WavePlayer_10
   WavePlayer_11
   WavePlayer_12
   WavePlayer_13
   WavePlayer_14
   WavePlayer_15
   WavePlayer_16
   WavePlayer_17
  ]
  properties {weights [1,1,1,1,1,1,1,1,1,1]}
 }
 {id=Random_2, children [WavePlayer_22, WavePlayer_23, WavePlayer_24], properties {weights [1,1,1]}}
 {id=Random_3, children [WavePlayer_19, WavePlayer_20, WavePlayer_21], properties {weights [1,1,1]}}
 {id=WavePlayer_1, properties {bLooping=true, soundWaveAssetPtr=&/Spot_ThemePark_Ambient_Max_Focus_Loop}}
 {id=WavePlayer_10, properties {soundWaveAssetPtr=&/Spot_Roller_Coaster_Bys_Distant_08}}
 {id=WavePlayer_11, properties {soundWaveAssetPtr=&/Spot_Roller_Coaster_Bys_Distant_07}}
 {id=WavePlayer_12, properties {soundWaveAssetPtr=&/Spot_Roller_Coaster_Bys_Distant_06}}
 {id=WavePlayer_13, properties {soundWaveAssetPtr=&/Spot_Roller_Coaster_Bys_Distant_05}}
 {id=WavePlayer_14, properties {soundWaveAssetPtr=&/Spot_Roller_Coaster_Bys_Distant_04}}
 {id=WavePlayer_15, properties {soundWaveAssetPtr=&/Spot_Roller_Coaster_Bys_Distant_03}}
 {id=WavePlayer_16, properties {soundWaveAssetPtr=&/Spot_Roller_Coaster_Bys_Distant_02}}
 {id=WavePlayer_17, properties {soundWaveAssetPtr=&/Spot_Roller_Coaster_Bys_Distant_01}}
 {id=WavePlayer_18, properties {bLooping=true, soundWaveAssetPtr=&Spot_ThemePark_Ambient_Max_Focus_Loop}}
 {id=WavePlayer_19, properties {soundWaveAssetPtr=&/Spot_Roller_Coaster_Bys_Distant_10}}
 {id=WavePlayer_2, properties {soundWaveAssetPtr=&/Spot_ThemePark_Organ_Oneshots_Distant_01}}
 {id=WavePlayer_20, properties {soundWaveAssetPtr=&/Spot_Roller_Coaster_Bys_Distant_09}}
 {id=WavePlayer_21, properties {soundWaveAssetPtr=&Spot_Roller_Coaster_Bys_Distant_08}}
 {id=WavePlayer_22, properties {soundWaveAssetPtr=&Spot_ThemePark_Organ_Oneshots_Distant_01}}
 {id=WavePlayer_23, properties {soundWaveAssetPtr=&/Spot_ThemePark_Organ_Oneshots_Distant_02}}
 {id=WavePlayer_24, properties {soundWaveAssetPtr=&/Spot_ThemePark_Organ_Oneshots_Distant_03}}
 {id=WavePlayer_3, properties {soundWaveAssetPtr=&Spot_ThemePark_Organ_Oneshots_Distant_02}}
 {id=WavePlayer_4, properties {soundWaveAssetPtr=&Spot_ThemePark_Organ_Oneshots_Distant_03}}
 {id=WavePlayer_5, properties {soundWaveAssetPtr=&/Spot_ThemePark_Organ_Oneshots_Distant_04}}
 {id=WavePlayer_6, properties {soundWaveAssetPtr=&/Spot_ThemePark_Organ_Oneshots_Distant_05}}
 {id=WavePlayer_7, properties {soundWaveAssetPtr=&/Spot_ThemePark_Organ_Oneshots_Distant_06}}
 {id=WavePlayer_8, properties {soundWaveAssetPtr=&Spot_Roller_Coaster_Bys_Distant_10}}
 {id=WavePlayer_9, properties {soundWaveAssetPtr=&Spot_Roller_Coaster_Bys_Distant_09}}
]
properties {
 bEnableSubmixSends=false
 bOverrideConcurrency=true
 maxDistance=25000
 soundClassObject=&Fort_Ambient_Outdoors_Athena
 subtitlePriority=5000
 volumeMultiplier=0.4
 attenuationOverrides {
  attenuationShape=Capsule
  bAttenuateWithLPF=true
  bEnableFocusInterpolation=true
  bEnableListenerFocus=true
  dBAttenuationAtMax=-40
  falloffDistance=2638.2017
  lPFFrequencyAtMax=1000
  lPFRadiusMax=3000
  lPFRadiusMin=1000
  nonFocusVolumeAttenuation=0
  occlusionInterpolationTime=1
  occlusionLowPassFilterFrequency=10000
  occlusionTraceChannel=ECC_GameTraceChannel11
  occlusionVolumeAttenuation=0.5
  stereoSpread=512
  attenuationShapeExtents=<3720.1087 1500 90>
 }
 busSends [{audioBus=null, minSendLevel=0, sendLevel=1, soundSourceBus=&AmbientAudio_Exterior_SourceBus}]
 concurrencyOverrides {maxCount=1, resolutionRule=PreventNew, voiceStealReleaseTime=2}
 soundSubmixSends [{minSendLevel=0, soundSubmix=null}]
}
