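# Sound cue for distant exterior calliope organ one-shots at the theme park POI, in the outdoor ambient sound class.
# A quality-level node routes its first two levels to a random pick of nine waves and its last two to a three-wave set, each through a modulator.
# Logarithmic attenuation adds distance filtering and curve-driven sends to a distant reverb submix.
# Concurrency allows two instances, stopping the farthest first.
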SoundCue /Hera/Audio/POI/ThemePark/Spots/Spot_ThemePark_Organ_Calliope_Distant_Ext_Cue
root=QualityLevel_0
nodes [
 {id=Modulator_0, children [Random_0], properties {pitchMin=1, volumeMax=1, volumeMin=1}}
 {id=Modulator_1, children [Random_3], properties {pitchMin=1, volumeMax=1, volumeMin=1}}
 {id=QualityLevel_0, children [Modulator_0,Modulator_0,Modulator_1,Modulator_1]}
 {
  id=Random_0
  children [WavePlayer_9, WavePlayer_10, WavePlayer_11, WavePlayer_12, WavePlayer_13, WavePlayer_14, WavePlayer_15, WavePlayer_16, WavePlayer_17]
  properties {weights [1,1,1,1,1,1,1,1,1]}
 }
 {id=Random_3, children [WavePlayer_22, WavePlayer_23, WavePlayer_21], properties {weights [1,1,1]}}
 {id=WavePlayer_10, properties {soundWaveAssetPtr=&/Spot_ThemePark_Organ_Calliope_02}}
 {id=WavePlayer_11, properties {soundWaveAssetPtr=&/Spot_ThemePark_Organ_Calliope_03}}
 {id=WavePlayer_12, properties {soundWaveAssetPtr=&/Spot_ThemePark_Organ_Calliope_04}}
 {id=WavePlayer_13, properties {soundWaveAssetPtr=&/Spot_ThemePark_Organ_Calliope_05}}
 {id=WavePlayer_14, properties {soundWaveAssetPtr=&/Spot_ThemePark_Organ_Calliope_06}}
 {id=WavePlayer_15, properties {soundWaveAssetPtr=&/Spot_ThemePark_Organ_Calliope_07}}
 {id=WavePlayer_16, properties {soundWaveAssetPtr=&/Spot_ThemePark_Organ_Calliope_08}}
 {id=WavePlayer_17, properties {soundWaveAssetPtr=&/Spot_ThemePark_Organ_Calliope_09}}
 {id=WavePlayer_21, properties {soundWaveAssetPtr=&Spot_ThemePark_Organ_Calliope_06}}
 {id=WavePlayer_22, properties {soundWaveAssetPtr=&/Spot_ThemePark_Organ_Calliope_01}}
 {id=WavePlayer_23, properties {soundWaveAssetPtr=&Spot_ThemePark_Organ_Calliope_02}}
 {id=WavePlayer_9, properties {soundWaveAssetPtr=&Spot_ThemePark_Organ_Calliope_01}}
]
properties {
 bOverrideAttenuation=true
 bOverrideConcurrency=true
 maxDistance=9000
 soundClassObject=&Fort_Ambient_OneShots_Outdoors
 attenuationOverrides {
  bAttenuateWithLPF=true
  bEnableLogFrequencyScaling=true
  bEnableSubmixSends=true
  distanceAlgorithm=Logarithmic
  falloffDistance=8000
  hPFFrequencyAtMax=500
  hPFFrequencyAtMin=100
  lPFFrequencyAtMax=500
  lPFRadiusMax=8000
  lPFRadiusMin=4000
  stereoSpread=2000
  attenuationShapeExtents {x=1000}
  submixSendSettings [{
    bEnableHPFCutoff=false
    bEnableLPFCutoff=false
    disableManualSendClamp=false
    hPFCutoff=20
    lPFCutoff=20000
    maxSendDistance=8000
    maxSendLevel=1
    minSendDistance=0
    minSendLevel=0
    sendLevel=0.2
    sendLevelControlMethod=CustomCurve
    soundSubmix=&Fortnite_Reverb_Distant
    customSendLevelCurve {editorCurveData {keys [<at 0 0>,<at 0.5 0>,<at 1 1>]}}
   }]
 }
 concurrencyOverrides {maxCount=2, resolutionRule=StopFarthestThenPreventNew}
 soundSubmixSends [{
   maxSendDistance=8000
   minSendLevel=0
   sendLevelControlMethod=CustomCurve
   soundSubmix=&Fortnite_Reverb_Distant
   customSendLevelCurve {editorCurveData {keys [<at 0 0>,<at 0.5 0>,<at 1 0.6>]}}
  }]
}
