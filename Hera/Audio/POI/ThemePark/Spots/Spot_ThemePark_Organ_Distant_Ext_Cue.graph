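# Sound Cue for distant exterior organ one-shots at the theme park POI, classed as outdoor ambient one-shots.
# A quality switch gives the first two levels a random pick of six waves with modulated volume, and the last two a smaller three-wave pool.
# Overridden attenuation uses logarithmic falloff and distance filtering, with curve-driven sends to a distant reverb submix that rise in the far range.
# Concurrency allows two voices, stopping the farthest first.
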SoundCue /Hera/Audio/POI/ThemePark/Spots/Spot_ThemePark_Organ_Distant_Ext_Cue
root=QualityLevel_0
nodes [
 {id=Modulator_0, children [Random_1], properties {pitchMin=1, volumeMax=1.2, volumeMin=1}}
 {id=Modulator_1, children [Random_2], properties {pitchMin=1, volumeMax=1, volumeMin=1}}
 {id=QualityLevel_0, children [Modulator_0,Modulator_0,Modulator_1,Modulator_1]}
 {id=Random_1, children [WavePlayer_0,WavePlayer_1,WavePlayer_2,WavePlayer_3,WavePlayer_4,WavePlayer_5], properties {weights [1,1,1,1,1,1]}}
 {id=Random_2, children [WavePlayer_6,WavePlayer_7,WavePlayer_8], properties {weights [1,1,1]}}
 {id=WavePlayer_0, properties {soundWaveAssetPtr=&/Spot_ThemePark_Organ_Oneshots_Distant_01}}
 {id=WavePlayer_1, properties {soundWaveAssetPtr=&/Spot_ThemePark_Organ_Oneshots_Distant_02}}
 {id=WavePlayer_2, properties {soundWaveAssetPtr=&/Spot_ThemePark_Organ_Oneshots_Distant_03}}
 {id=WavePlayer_3, properties {soundWaveAssetPtr=&/Spot_ThemePark_Organ_Oneshots_Distant_04}}
 {id=WavePlayer_4, properties {soundWaveAssetPtr=&/Spot_ThemePark_Organ_Oneshots_Distant_05}}
 {id=WavePlayer_5, properties {soundWaveAssetPtr=&/Spot_ThemePark_Organ_Oneshots_Distant_06}}
 {id=WavePlayer_6, properties {soundWaveAssetPtr=&Spot_ThemePark_Organ_Oneshots_Distant_01}}
 {id=WavePlayer_7, properties {soundWaveAssetPtr=&Spot_ThemePark_Organ_Oneshots_Distant_02}}
 {id=WavePlayer_8, properties {soundWaveAssetPtr=&Spot_ThemePark_Organ_Oneshots_Distant_03}}
]
properties {
 bOverrideAttenuation=true
 bOverrideConcurrency=true
 maxDistance=9000
 soundClassObject=&Fort_Ambient_OneShots_Outdoors
 volumeMultiplier=1
 attenuationOverrides {
  bAttenuateWithLPF=true
  bEnableLogFrequencyScaling=true
  bEnableSubmixSends=true
  distanceAlgorithm=Logarithmic
  falloffDistance=8000
  hPFFrequencyAtMax=500
  hPFFrequencyAtMin=100
  lPFFrequencyAtMax=500
  lPFRadiusMax=8000
  lPFRadiusMin=4000
  stereoSpread=2000
  attenuationShapeExtents {x=1000}
  submixSendSettings [{
    bEnableHPFCutoff=false
    bEnableLPFCutoff=false
    disableManualSendClamp=false
    hPFCutoff=20
    lPFCutoff=20000
    maxSendDistance=8000
    maxSendLevel=1
    minSendDistance=0
    minSendLevel=0
    sendLevel=0.2
    sendLevelControlMethod=CustomCurve
    soundSubmix=&Fortnite_Reverb_Distant
    customSendLevelCurve {editorCurveData {keys [<at 0 0>,<at 0.5 0>,<at 1 1>]}}
   }]
 }
 concurrencyOverrides {maxCount=2, resolutionRule=StopFarthestThenPreventNew}
 soundSubmixSends [{
   maxSendDistance=8000
   minSendLevel=0
   sendLevelControlMethod=CustomCurve
   soundSubmix=&Fortnite_Reverb_Distant
   customSendLevelCurve {editorCurveData {keys [<at 0 0>,<at 0.5 0>,<at 1 0.6>]}}
  }]
}
